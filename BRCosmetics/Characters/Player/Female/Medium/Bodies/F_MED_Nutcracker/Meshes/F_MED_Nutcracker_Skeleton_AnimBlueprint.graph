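# Post-style cosmetic AnimBlueprint for the female medium Nutcracker body that poses its dynamic shoulder pad bones to follow the arms.
# Copy Pose From Mesh takes the attached parent's pose, then feeds left and right Pose Drivers in series to the output.
# Each RBF driver reads clavicle and upper arm bones, blending Base and ADS forward, down and up targets from a pistol pose asset onto that side's pad bones.
# The event graph has no logic.
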AnimBlueprint /BRCosmetics/Characters/Player/Female/Medium/Bodies/F_MED_Nutcracker/Meshes/F_MED_Nutcracker_Skeleton_AnimBlueprint
target_skeleton=&/F_MED_Nutcracker_Skeleton
graphs [
 AnimGraph {nodes [
   n1 = Pose Driver - Source: upperarm_l!(SourcePose:Pose Link Structure=Copy Pose From Mesh() {bUseAttachedParent=true}) -> (Pose:Pose Link Structure) {
    is=title AnimGraphNode_PoseDriver
    poseAsset=&/F_MED_Nutcracker_Pistol_RBF_PoseAsset
    onlyDriveBones [{boneName=dyn_shoulderPad_master_l}, {boneName=dyn_shoulderPad_l}, {boneName=dyn_shoulderPad_front_l}]
    poseTargets [
     {
      bApplyCustomCurve=false
      bIsHidden=false
      distanceMethod=DefaultMethod
      drivenName=Base
      functionType=DefaultFunction
      targetScale=0.5527
      boneTransforms [
       {targetRotation=<rot -58.1985 2.4027 1.3901>, targetTranslation=<9.8658 -1.2136 -0.4105>}
       {targetRotation=<rot -85.1915 -138.5625 143.0619>, targetTranslation=<5.4831 -0.1053 -3.2946>}
      ]
      customCurve {defaultValue=3.4028234663852886e+38, postInfinityExtrap=RCCE_Constant, preInfinityExtrap=RCCE_Constant, keys []}
      targetRotation=<rot 0 0 0>
     }
     {
      bApplyCustomCurve=false
      bIsHidden=false
      distanceMethod=DefaultMethod
      drivenName=ADS_Forward
      functionType=DefaultFunction
      targetScale=0.957
      boneTransforms [
       {targetRotation=<rot -31.3117 55.2109 -76.9664>, targetTranslation=<9.8658 -1.2136 -0.4105>}
       {targetRotation=<rot -77.6267 47.311 -43.7067>, targetTranslation=<5.4831 -0.1053 -3.2946>}
      ]
      customCurve {defaultValue=3.4028234663852886e+38, postInfinityExtrap=RCCE_Constant, preInfinityExtrap=RCCE_Constant, keys []}
      targetRotation=<rot 0 0 0>
     }
     {
      bApplyCustomCurve=false
      bIsHidden=false
      distanceMethod=DefaultMethod
      drivenName=ADS_Down
      functionType=DefaultFunction
      targetScale=0.5527
      boneTransforms [
       {targetRotation=<rot -78.6228 44.2493 -27.3753>, targetTranslation=<9.8658 -1.2136 -0.4105>}
       {targetRotation=<rot -80.948 90.9335 -88.3662>, targetTranslation=<5.4831 -0.1053 -3.2946>}
      ]
      customCurve {defaultValue=3.4028234663852886e+38, postInfinityExtrap=RCCE_Constant, preInfinityExtrap=RCCE_Constant, keys []}
      targetRotation=<rot 0 0 0>
     }
     {
      bApplyCustomCurve=false
      bIsHidden=false
      distanceMethod=DefaultMethod
      drivenName=ADS_Up
      functionType=DefaultFunction
      targetScale=1
      boneTransforms [
       {targetRotation=<rot 6.5607 44.7836 -105.5724>, targetTranslation=<9.8658 -1.2136 -0.4105>}
       {targetRotation=<rot -49.177 16.469 -6.257>, targetTranslation=<5.4831 -0.1053 -3.2946>}
      ]
      customCurve {defaultValue=3.4028234663852886e+38, postInfinityExtrap=RCCE_Constant, preInfinityExtrap=RCCE_Constant, keys []}
      targetRotation=<rot 0 0 0>
     }
    ]
    rBFParams {radius=24.8952}
    sourceBones [{boneName=upperarm_l}, {boneName=clavicle_l}]
   }
   n2 = Pose Driver - Source: clavicle_r!(SourcePose:Pose Link Structure=@n1) -> (Pose:Pose Link Structure) {
    is=title AnimGraphNode_PoseDriver
    poseAsset=&F_MED_Nutcracker_Pistol_RBF_PoseAsset
    onlyDriveBones [{boneName=dyn_shoulderPad_master_r}, {boneName=dyn_shoulderPad_r}, {boneName=dyn_shoulderPad_front_r}]
    poseTargets [
     {
      bApplyCustomCurve=false
      bIsHidden=false
      distanceMethod=DefaultMethod
      drivenName=Base
      functionType=DefaultFunction
      targetScale=1
      boneTransforms [
       {targetRotation=<rot -85.1915 41.4375 143.0619>, targetTranslation=<5.4831 -0.1053 3.2946>}
       {targetRotation=<rot -58.1985 2.4027 1.3901>, targetTranslation=<-9.8658 1.2136 0.4104>}
      ]
      customCurve {defaultValue=3.4028234663852886e+38, postInfinityExtrap=RCCE_Constant, preInfinityExtrap=RCCE_Constant, keys []}
      targetRotation=<rot 0 0 0>
     }
     {
      bApplyCustomCurve=false
      bIsHidden=false
      distanceMethod=DefaultMethod
      drivenName=ADS_Forward
      functionType=DefaultFunction
      targetScale=0.6226
      boneTransforms [
       {targetRotation=<rot -68.9944 -127.6269 -53.9154>, targetTranslation=<5.4831 -0.1053 3.2946>}
       {targetRotation=<rot -3.8427 98.818 -54.4446>, targetTranslation=<-9.8658 1.2136 0.4104>}
      ]
      customCurve {defaultValue=3.4028234663852886e+38, postInfinityExtrap=RCCE_Constant, preInfinityExtrap=RCCE_Constant, keys []}
      targetRotation=<rot 0 0 0>
     }
     {
      bApplyCustomCurve=false
      bIsHidden=false
      distanceMethod=DefaultMethod
      drivenName=ADS_Down
      functionType=DefaultFunction
      targetScale=0.8636
      boneTransforms [
       {targetRotation=<rot -71.55 -117.4208 -63.5341>, targetTranslation=<5.4831 -0.1053 3.2946>}
       {targetRotation=<rot -51.8404 98.8831 -31.0819>, targetTranslation=<-9.8658 1.2136 0.4104>}
      ]
      customCurve {defaultValue=3.4028234663852886e+38, postInfinityExtrap=RCCE_Constant, preInfinityExtrap=RCCE_Constant, keys []}
      targetRotation=<rot 0 0 0>
     }
     {
      bApplyCustomCurve=false
      bIsHidden=false
      distanceMethod=DefaultMethod
      drivenName=ADS_Up
      functionType=DefaultFunction
      targetScale=0.6226
      boneTransforms [
       {targetRotation=<rot -68.9927 -127.6305 -53.9082>, targetTranslation=<5.4831 -0.1053 3.2946>}
       {targetRotation=<rot 29.9686 90.4741 -81.76>, targetTranslation=<-9.8658 1.2136 0.4104>}
      ]
      customCurve {defaultValue=3.4028234663852886e+38, postInfinityExtrap=RCCE_Constant, preInfinityExtrap=RCCE_Constant, keys []}
      targetRotation=<rot 0 0 0>
     }
    ]
    rBFParams {radius=40}
    sourceBones [{boneName=clavicle_r}, {boneName=upperarm_r}]
   }
   Output Pose(Result=@n2) {layerGroup=None}
  ]}
 EventGraph {nodes [Event BlueprintUpdateAnimation()]}
]
variables []
