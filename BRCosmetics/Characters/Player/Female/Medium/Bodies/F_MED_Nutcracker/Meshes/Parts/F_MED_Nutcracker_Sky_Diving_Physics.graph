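# Physics asset for the sky-diving variant of the female medium Nutcracker cosmetic, covering the upper body and dynamic hair.
# Capsules form the head, neck, upper spine and clavicle bodies, joined by constraints with free swing and twist.
# Two mirrored three-link chains of thin box bodies hang from the head for left and right hair.
# Hair constraints allow almost no twist, and their swing limits tighten toward the tips, so the chain ends stay stiffest.
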PhysicsAsset /BRCosmetics/Characters/Player/Female/Medium/Bodies/F_MED_Nutcracker/Meshes/Parts/F_MED_Nutcracker_Sky_Diving_Physics
bodies {
 clavicle_l [{shapeType=Capsule, length=7.5, radius=10, center=<8.6 0 2>, rotation=<rot 60 180 180>}]
 clavicle_r [{shapeType=Capsule, length=7.5, radius=10, center=<-8.6 0 -2>, rotation=<rot 60 180 -180>}]
 dyn_hair_01_l [{shapeType=Box, extentX=10, extentY=2, extentZ=10, center=<5 0 0>}]
 dyn_hair_01_r [{shapeType=Box, extentX=10, extentY=2, extentZ=10, center=<5 0 0>}]
 dyn_hair_02_l [{shapeType=Box, extentX=10, extentY=2, extentZ=10, center=<5 0 0>}]
 dyn_hair_02_r [{shapeType=Box, extentX=10, extentY=2, extentZ=10, center=<5 0 0>}]
 dyn_hair_03_l [{shapeType=Box, extentX=10, extentY=2, extentZ=10, center=<5 0 0>}]
 dyn_hair_03_r [{shapeType=Box, extentX=10, extentY=2, extentZ=10, center=<5 0 0>}]
 head [{shapeType=Capsule, length=3.75, radius=9.25, center=<1.0589 1.6885 0>, rotation=<rot 90 0 0>}]
 neck_01 [{shapeType=Capsule, length=5.25, radius=7.5, center=<1.3587 0.0717 0>}]
 neck_02 [{shapeType=Capsule, length=0, radius=4.4279, center=<3.169 -0.4539 0>, rotation=<rot 90 0 0>}]
 spine_05 [{shapeType=Capsule, length=3.8108, radius=11.7085}]
}
constraints [
 {
  bone1Name=head
  bone2Name=neck_02
  swing1LimitDegrees=45
  swing1Motion=Free
  swing2LimitDegrees=45
  swing2Motion=Free
  twistLimitDegrees=45
  twistMotion=Free
 }
 {
  bone1Name=neck_02
  bone2Name=neck_01
  swing1LimitDegrees=45
  swing1Motion=Free
  swing2LimitDegrees=45
  swing2Motion=Free
  twistLimitDegrees=45
  twistMotion=Free
 }
 {
  bone1Name=neck_01
  bone2Name=spine_05
  swing1LimitDegrees=45
  swing1Motion=Free
  swing2LimitDegrees=45
  swing2Motion=Free
  twistLimitDegrees=45
  twistMotion=Free
 }
 {
  bone1Name=clavicle_r
  bone2Name=spine_05
  swing1LimitDegrees=45
  swing1Motion=Free
  swing2LimitDegrees=45
  swing2Motion=Free
  twistLimitDegrees=45
  twistMotion=Free
 }
 {
  bone1Name=clavicle_l
  bone2Name=spine_05
  swing1LimitDegrees=45
  swing1Motion=Free
  swing2LimitDegrees=45
  swing2Motion=Free
  twistLimitDegrees=45
  twistMotion=Free
 }
 {bone1Name=dyn_hair_01_l, bone2Name=head, swing1LimitDegrees=30, swing2LimitDegrees=20, twistLimitDegrees=1}
 {bone1Name=dyn_hair_02_l, bone2Name=dyn_hair_01_l, swing1LimitDegrees=30, swing2LimitDegrees=15, twistLimitDegrees=1}
 {bone1Name=dyn_hair_03_l, bone2Name=dyn_hair_02_l, swing1LimitDegrees=5, swing2LimitDegrees=5, twistLimitDegrees=1}
 {bone1Name=dyn_hair_01_r, bone2Name=head, swing1LimitDegrees=30, swing2LimitDegrees=20, twistLimitDegrees=1}
 {bone1Name=dyn_hair_02_r, bone2Name=dyn_hair_01_r, swing1LimitDegrees=30, swing2LimitDegrees=15, twistLimitDegrees=1}
 {bone1Name=dyn_hair_03_r, bone2Name=dyn_hair_02_r, swing1LimitDegrees=5, swing2LimitDegrees=5, twistLimitDegrees=1}
]
