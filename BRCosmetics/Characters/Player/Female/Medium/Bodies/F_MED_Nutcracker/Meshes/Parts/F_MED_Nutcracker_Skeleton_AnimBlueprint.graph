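# Post-process-style hair physics AnimBlueprint for the female Nutcracker body, layered on the pose copied from the attached parent mesh.
# A bool switches between normal and skydiving head-relative RigidBody simulations; the latter takes a state-based gravity override from the EventGraph as external force.
# Jumping or falling blends in hair bone offsets, then constraints half-weight left hair rotations toward the right.
# An AnimDynamics chain remains unconnected to the output.
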AnimBlueprint /BRCosmetics/Characters/Player/Female/Medium/Bodies/F_MED_Nutcracker/Meshes/Parts/F_MED_Nutcracker_Skeleton_AnimBlueprint
target_skeleton=&/F_MED_Nutcracker_Skeleton
graphs [
 AnimGraph {nodes [
   n1 = Break Fort Anim Input Skydiving External Force(FortAnimInput_SkydivingExternalForce=Get SkydivingExternalForceInput()) -> (
    bUseSkydivingVectorForce=~
    SkydivingVectorForce=~
    HeadToPelvisDirection=~
    FloatingMultiplier=~
    FloatingAdditive=~
    DivingMultiplier=~
    DivingAdditive=~
    ParachutingMultiplier=~
    ParachutingAdditive=~
   )
   n2 = Anim Dynamics - Bone: dyn_hair_01_r(Alpha=1, ExternalForce=@n1.SkydivingVectorForce) {
    angularDampingOverride=0.9999
    bGravityOverrideInSimSpace=true
    bOverrideAngularDamping=true
    bUsePlanarLimit=false
    alphaBoolBlend {blendOption=Linear}
    boundBone {boneName=dyn_hair_01_r}
    componentAppliedLinearAccClamp=<10 10 10>
    componentLinearAccScale=<1 1 1>
    componentLinearVelScale=<1 1 1>
    physicsBodyDefinitions [{
      collisionType=CustomSphere
      sphereCollisionRadius=1
      boundBone {boneName=dyn_hair_01_r}
      boxExtents=<5 20 20>
      constraintSetup {angularLimitsMax=<0 10 0>, angularLimitsMin=<0 -10 -40>, angularTarget=<0 0 0>}
      localJointOffset=<5 0 0>
     }]
    planarLimits [{drivingBone {boneName=spine_05}, planeTransform {location=<0 -10 0>, rotation=<rot 0 30 -89.9999>}}]
   }
   n3 = Anim Dynamics - Bone: dyn_hair_01_l(ComponentPose=@n2, Alpha=1, ExternalForce=@n1.SkydivingVectorForce) {
    angularDampingOverride=0.9999
    bGravityOverrideInSimSpace=true
    bOverrideAngularDamping=true
    bUsePlanarLimit=false
    alphaBoolBlend {blendOption=Linear}
    boundBone {boneName=dyn_hair_01_l}
    componentAppliedLinearAccClamp=<10 10 10>
    componentLinearAccScale=<1 1 1>
    componentLinearVelScale=<1 1 1>
    physicsBodyDefinitions [{
      collisionType=CustomSphere
      sphereCollisionRadius=1
      boundBone {boneName=dyn_hair_01_l}
      boxExtents=<5 20 20>
      constraintSetup {angularLimitsMax=<0 10 0>, angularLimitsMin=<0 -10 -40>, angularTarget=<0 0 0>}
      localJointOffset=<5 0 0>
     }]
    planarLimits [{drivingBone {boneName=spine_05}, planeTransform {location=<0 -10 0>, rotation=<rot 0 30 -89.9999>}}]
   }
   n4 = Anim Dynamics - Bone: dyn_hair_02_r(ComponentPose=@n3, Alpha=1, ExternalForce=@n1.SkydivingVectorForce) {
    angularDampingOverride=0.9999
    bGravityOverrideInSimSpace=true
    bOverrideAngularDamping=true
    bUsePlanarLimit=false
    alphaBoolBlend {blendOption=Linear}
    boundBone {boneName=dyn_hair_02_r}
    componentAppliedLinearAccClamp=<10 10 10>
    componentLinearAccScale=<1 1 1>
    componentLinearVelScale=<1 1 1>
    physicsBodyDefinitions [{
      collisionType=CustomSphere
      sphereCollisionRadius=1
      boundBone {boneName=dyn_hair_02_r}
      boxExtents=<5 20 20>
      constraintSetup {angularLimitsMax=<0 0 0>, angularLimitsMin=<0 -10 -40>, angularTarget=<0 0 0>}
      localJointOffset=<5 0 0>
     }]
    planarLimits [{drivingBone {boneName=spine_05}, planeTransform {location=<0 -10 0>, rotation=<rot 0 30 -89.9999>}}]
   }
   Anim Dynamics - Bone: dyn_hair_02_l(ComponentPose=@n4, Alpha=1, ExternalForce=@n1.SkydivingVectorForce) {
    angularDampingOverride=0.9999
    bGravityOverrideInSimSpace=true
    bOverrideAngularDamping=true
    bUsePlanarLimit=false
    alphaBoolBlend {blendOption=Linear}
    boundBone {boneName=dyn_hair_02_l}
    componentAppliedLinearAccClamp=<10 10 10>
    componentLinearAccScale=<1 1 1>
    componentLinearVelScale=<1 1 1>
    physicsBodyDefinitions [{
      collisionType=CustomSphere
      sphereCollisionRadius=1
      boundBone {boneName=dyn_hair_02_l}
      boxExtents=<5 20 20>
      constraintSetup {angularLimitsMax=<0 10 0>, angularLimitsMin=<0 -10 -40>, angularTarget=<0 0 0>}
      localJointOffset=<5 0 0>
     }]
    planarLimits [{drivingBone {boneName=spine_05}, planeTransform {location=<0 -10 0>, rotation=<rot 0 30 -89.9999>}}]
   }
   "dyn_hair_01_l.rotateZ = head.rotateZ - Bone Driven Controller"!(ComponentPose:Component Space Pose Link Structure, Alpha:float=1) -> (Pose:Component Space Pose Link Structure) {
    is=title AnimGraphNode_BoneDrivenController
    bAffectTargetRotationZ=true
    sourceComponent=RotationZ
    alphaBoolBlend {blendOption=Linear}
    sourceBone {boneName=head}
    targetBone {boneName=dyn_hair_01_l}
   }
   n5 = "dyn_hair_master.rotateX = neck_01.rotateX * -1 - Bone Driven Controller"!(
    ComponentPose:Component Space Pose Link Structure=Local To Component(LocalPose=Copy Pose From Mesh() {bUseAttachedParent=true})
    Alpha:float=1
   ) -> (Pose:Component Space Pose Link Structure) {
    is=title AnimGraphNode_BoneDrivenController
    bAffectTargetRotationX=true
    multiplier=-1
    sourceComponent=RotationX
    alphaBoolBlend {blendOption=Linear}
    sourceBone {boneName=neck_01}
    targetBone {boneName=dyn_hair_master}
   }
   n6 = RigidBody(ComponentPose=@n5) {
    bClampLinearTranslationLimitToRefPose=true
    bOverrideWorldGravity=true
    overridePhysicsAsset=&/F_MED_Nutcracker_Physics
    simulationSpace=BaseBoneSpace
    alphaBoolBlend {blendOption=Linear}
    baseBoneRef {boneName=head}
    componentAppliedLinearAccClamp=<5000 5000 5000>
    componentLinearAccScale=<0.4 0.4 0.4>
    overrideWorldGravity {z=-980}
   }
   n7 = "Transform (Modify) Bone - Bone: dyn_hair_master"!(ComponentPose=@n6, bAlphaBoolEnabled:bool=True, Translation="1,-1,0", Rotation="0,2,0", Scale="1,1,1") -> (Pose) {
    alphaInputType=Bool
    rotationMode=BMM_Replace
    rotationSpace=BCS_BoneSpace
    translationMode=BMM_Replace
    translationSpace=BCS_BoneSpace
    alphaBoolBlend {blendOption=Linear}
    boneToModify {boneName=dyn_hair_master}
    rotation {yaw=2}
    translation=<1 -1>
   }
   n8 = Component To Local(ComponentPose=@n7)
   n9 = RigidBody(ComponentPose=Local To Component(LocalPose=Copy Pose From Mesh() {bUseAttachedParent=true}), ExternalForce=Get Gravity Override()) {
    bClampLinearTranslationLimitToRefPose=true
    bOverrideWorldGravity=true
    overridePhysicsAsset=&/F_MED_Nutcracker_Sky_Diving_Physics
    simulationSpace=BaseBoneSpace
    alphaBoolBlend {blendOption=Linear}
    baseBoneRef {boneName=head}
    componentAppliedLinearAccClamp=<3000 3000 3000>
    componentLinearAccScale=<0.3 0.3 0.45>
    componentLinearVelScale=<0.15 0.15 0.225>
    overrideWorldGravity {z=-980}
   }
   n10 = "Transform (Modify) Bone - Bone: dyn_hair_master"(ComponentPose=@n9, bAlphaBoolEnabled=True, Translation="1,-1,0", Rotation="0,15,0", Scale="1,1,1") {
    alphaInputType=Bool
    rotationMode=BMM_Replace
    rotationSpace=BCS_BoneSpace
    translationMode=BMM_Replace
    translationSpace=BCS_BoneSpace
    alphaBoolBlend {blendOption=Linear}
    boneToModify {boneName=dyn_hair_master}
    rotation {yaw=15}
    translation=<1 -1>
   }
   n11 = "Transform (Modify) Bone - Bone: dyn_hair_02_l"!(ComponentPose=@n10, bAlphaBoolEnabled:bool=True, Translation="0,0,0", Rotation="25,0,0", Scale="1,1,1") -> (Pose) {
    alphaInputType=Bool
    rotationMode=BMM_Replace
    rotationSpace=BCS_BoneSpace
    translationMode=BMM_Replace
    translationSpace=BCS_BoneSpace
    alphaBoolBlend {blendOption=Linear}
    boneToModify {boneName=dyn_hair_02_l}
    rotation {pitch=25}
   }
   n12 = "Transform (Modify) Bone - Bone: dyn_hair_02_r"!(ComponentPose=@n11, bAlphaBoolEnabled:bool=True, Translation="0,0,0", Rotation="-25,0,0", Scale="1,1,1") -> (Pose) {
    alphaInputType=Bool
    rotationMode=BMM_Replace
    rotationSpace=BCS_BoneSpace
    translationMode=BMM_Replace
    translationSpace=BCS_BoneSpace
    alphaBoolBlend {blendOption=Linear}
    boneToModify {boneName=dyn_hair_02_r}
    rotation {pitch=-25}
   }
   n13 = Component To Local(ComponentPose=@n12)
   n14 = Get bIsSkyDiving()
   n15 = OR Boolean(A=Get bIsJumping(), B=Get bIsFalling())
   n16 = "Transform (Modify) Bone - Bone: dyn_hair_02_l"(
    ComponentPose=Local To Component(LocalPose=Blend Poses by bool(BlendPose_0=@n13, BlendPose_1=@n8, bActiveValue=@n14) {blendType=Linear})
    bAlphaBoolEnabled=@n15
    Translation="0,0,0"
    Rotation="0,5,0"
    Scale="1,1,1"
   ) {
    alphaInputType=Bool
    rotationMode=BMM_Replace
    rotationSpace=BCS_BoneSpace
    translationMode=BMM_Replace
    translationSpace=BCS_BoneSpace
    alphaBoolBlend {blendInTime=0.2, blendOption=Linear, blendOutTime=0.2}
    boneToModify {boneName=dyn_hair_02_l}
    rotation {yaw=5}
   }
   n17 = "Transform (Modify) Bone - Bone: dyn_hair_02_r"(ComponentPose=@n16, bAlphaBoolEnabled=@n15, Translation="0,0,0", Rotation="0,5,0", Scale="1,1,1") {
    alphaInputType=Bool
    rotationMode=BMM_Replace
    rotationSpace=BCS_BoneSpace
    translationMode=BMM_Replace
    translationSpace=BCS_BoneSpace
    alphaBoolBlend {blendInTime=0.2, blendOption=Linear, blendOutTime=0.2}
    boneToModify {boneName=dyn_hair_02_r}
    rotation {yaw=5}
   }
   n18 = "Transform (Modify) Bone - Bone: dyn_hair_03_l"!(ComponentPose=@n17, bAlphaBoolEnabled:bool=@n15, Translation="0,0,0", Rotation="0,-20,0", Scale="1,1,1") -> (Pose) {
    alphaInputType=Bool
    rotationMode=BMM_Replace
    rotationSpace=BCS_BoneSpace
    translationMode=BMM_Replace
    translationSpace=BCS_BoneSpace
    alphaBoolBlend {blendInTime=0.2, blendOption=Linear, blendOutTime=0.2}
    boneToModify {boneName=dyn_hair_03_l}
    rotation {yaw=-20}
   }
   n19 = "Transform (Modify) Bone - Bone: dyn_hair_03_r"!(ComponentPose=@n18, bAlphaBoolEnabled:bool=@n15, Translation="0,0,0", Rotation="0,-20,0", Scale="1,1,1") -> (Pose) {
    alphaInputType=Bool
    rotationMode=BMM_Replace
    rotationSpace=BCS_BoneSpace
    translationMode=BMM_Replace
    translationSpace=BCS_BoneSpace
    alphaBoolBlend {blendInTime=0.2, blendOption=Linear, blendOutTime=0.2}
    boneToModify {boneName=dyn_hair_03_r}
    rotation {yaw=-20}
   }
   n20 = "Constraint - dyn_hair_01_l (R)"(ComponentPose=@n19, Alpha=0.5, ConstraintWeights_0=1) {
    alphaBoolBlend {blendOption=Linear}
    boneToModify {boneName=dyn_hair_01_l}
    constraintSetup [{transformType=Rotation, targetBone {boneName=dyn_hair_01_r}}]
    constraintWeights [1]
   }
   n21 = "Constraint - dyn_hair_02_l (R)"(ComponentPose=@n20, Alpha=0.5, ConstraintWeights_0=1) {
    alphaBoolBlend {blendOption=Linear}
    boneToModify {boneName=dyn_hair_02_l}
    constraintSetup [{transformType=Rotation, targetBone {boneName=dyn_hair_02_r}}]
    constraintWeights [1]
   }
   n22 = "Constraint - dyn_hair_03_l (R)"(ComponentPose=@n21, Alpha=0.5, ConstraintWeights_0=1) {
    alphaBoolBlend {blendOption=Linear}
    boneToModify {boneName=dyn_hair_03_l}
    constraintSetup [{transformType=Rotation, targetBone {boneName=dyn_hair_03_r}}]
    constraintWeights [1]
   }
   Output Pose(Result=Component To Local(ComponentPose=@n22)) {layerGroup=None}
  ]}
 EventGraph {nodes [
   Event BlueprintUpdateAnimation()
   > n1 = Gravity Override State Function DEPRECATED(CharacterPartAnimInstance=Self-Reference(), GameTimeInSeconds=GetGameTimeInSeconds(), GravityOverrideParams=Get Gravity Override Params())
   > Set Gravity Override(Gravity Override=@n1)
   TryGetPawnOwner()
  ]}
]
variables [
 Gravity Override {type=Vector}
 Gravity Override Params {
  type=GravityOverrideParamsStruct
  default {
   skydive_DivingDown {gravityOverride=<0 0 200>}
   skydive_FloatingDown {gravityOverride=<0 0 100>}
   skydive_GoingUp {gravityOverride=<0 0 -980>}
   swimming {gravityOverride=<0 0 0>}
  }
 }
]
